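# Physics asset for the cape face-accessory mesh on the BLK variant of the male medium Grasshopper player cosmetic, with mirrored left and right flap chains.
# Capsules cover the upper spine, while tiny capsules mark the clavicles and each flap's MSTR root under spine_05.
# Each four-segment flap starts with a box and continues in capsules.
# Spine, clavicle and root joints swing and twist freely, while flap joints have angular limits.
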PhysicsAsset /BRCosmetics/Characters/Player/Male/Medium/Bodies/M_MED_Grasshopper/Meshes/Parts/BLK/M_MED_Grasshopper_BLK_FaceAcc_cape_Physics
bodies {
 clavicle_l [{shapeType=Capsule, length=0.505, radius=0.505, shapeName=clavicle_l_capsule}]
 clavicle_r [{shapeType=Capsule, length=0.505, radius=0.505, shapeName=clavicle_r_capsule}]
 dyn_flap_l_01 [{shapeType=Box, extentX=7.5, extentY=2.5, extentZ=10, shapeName=dyn_flap_l_01_box, center=<3 0 0>}]
 dyn_flap_l_01_MSTR [{shapeType=Capsule, length=0.505, radius=0.505, shapeName=dyn_flap_l_01_MSTR_capsule}]
 dyn_flap_l_02 [{shapeType=Capsule, length=6, radius=2, shapeName=dyn_flap_l_02_capsule, center=<4 0 0>}]
 dyn_flap_l_03 [{shapeType=Capsule, length=5, radius=2, shapeName=dyn_flap_l_03_capsule, center=<3 0 0>}]
 dyn_flap_l_04 [{shapeType=Capsule, length=4, radius=2, shapeName=dyn_flap_l_04_capsule, center=<3 0 0>, rotation=<rot 90 0 0>}]
 dyn_flap_r_01 [{shapeType=Box, extentX=7.5, extentY=2.5, extentZ=10, shapeName=dyn_flap_l_01_box, center=<3 0 0>}]
 dyn_flap_r_01_MSTR [{shapeType=Capsule, length=0.505, radius=0.505, shapeName=dyn_flap_r_01_MSTR_capsule}]
 dyn_flap_r_02 [{shapeType=Capsule, length=6, radius=2, shapeName=dyn_flap_l_02_capsule, center=<4 0 0>}]
 dyn_flap_r_03 [{shapeType=Capsule, length=5, radius=2, shapeName=dyn_flap_l_03_capsule, center=<3 0 0>}]
 dyn_flap_r_04 [{shapeType=Capsule, length=4, radius=2, shapeName=dyn_flap_l_04_capsule, center=<3 0 0>, rotation=<rot 90 0 0>}]
 spine_03 [{shapeType=Capsule, length=8, radius=10.2, shapeName=spine_03_capsule, center=<5 0 0>}]
 spine_04 [{shapeType=Capsule, length=10, radius=10.2, shapeName=spine_04_capsule, center=<5 0 0>}]
 spine_05 [{shapeType=Capsule, length=15, radius=9.2, shapeName=spine_05_capsule, center=<2 0 0>}]
}
constraints [
 {
  bone1Name=spine_04
  bone2Name=spine_03
  swing1LimitDegrees=45
  swing1Motion=Free
  swing2LimitDegrees=45
  swing2Motion=Free
  twistLimitDegrees=45
  twistMotion=Free
 }
 {
  bone1Name=spine_05
  bone2Name=spine_04
  swing1LimitDegrees=45
  swing1Motion=Free
  swing2LimitDegrees=45
  swing2Motion=Free
  twistLimitDegrees=45
  twistMotion=Free
 }
 {
  bone1Name=clavicle_l
  bone2Name=spine_05
  swing1LimitDegrees=45
  swing1Motion=Free
  swing2LimitDegrees=45
  swing2Motion=Free
  twistLimitDegrees=45
  twistMotion=Free
 }
 {
  bone1Name=clavicle_r
  bone2Name=spine_05
  swing1LimitDegrees=45
  swing1Motion=Free
  swing2LimitDegrees=45
  swing2Motion=Free
  twistLimitDegrees=45
  twistMotion=Free
 }
 {
  bone1Name=dyn_flap_l_01_MSTR
  bone2Name=spine_05
  swing1LimitDegrees=45
  swing1Motion=Free
  swing2LimitDegrees=45
  swing2Motion=Free
  twistLimitDegrees=45
  twistMotion=Free
 }
 {bone1Name=dyn_flap_l_01, bone2Name=dyn_flap_l_01_MSTR, swing1LimitDegrees=15, swing2LimitDegrees=5, twistLimitDegrees=5}
 {bone1Name=dyn_flap_l_02, bone2Name=dyn_flap_l_01, swing1LimitDegrees=45, swing2LimitDegrees=15, twistLimitDegrees=15}
 {bone1Name=dyn_flap_l_03, bone2Name=dyn_flap_l_02, swing1LimitDegrees=30, swing2LimitDegrees=30, twistLimitDegrees=15}
 {bone1Name=dyn_flap_l_04, bone2Name=dyn_flap_l_03, swing1LimitDegrees=30, swing2LimitDegrees=30, twistLimitDegrees=15}
 {
  bone1Name=dyn_flap_r_01_MSTR
  bone2Name=spine_05
  swing1LimitDegrees=45
  swing1Motion=Free
  swing2LimitDegrees=45
  swing2Motion=Free
  twistLimitDegrees=45
  twistMotion=Free
 }
 {bone1Name=dyn_flap_r_01, bone2Name=dyn_flap_r_01_MSTR, swing1LimitDegrees=15, swing2LimitDegrees=5, twistLimitDegrees=5}
 {bone1Name=dyn_flap_r_02, bone2Name=dyn_flap_r_01, swing1LimitDegrees=45, swing2LimitDegrees=15, twistLimitDegrees=15}
 {bone1Name=dyn_flap_r_03, bone2Name=dyn_flap_r_02, swing1LimitDegrees=30, swing2LimitDegrees=30, twistLimitDegrees=15}
 {bone1Name=dyn_flap_r_04, bone2Name=dyn_flap_r_03, swing1LimitDegrees=30, swing2LimitDegrees=30, twistLimitDegrees=15}
]
